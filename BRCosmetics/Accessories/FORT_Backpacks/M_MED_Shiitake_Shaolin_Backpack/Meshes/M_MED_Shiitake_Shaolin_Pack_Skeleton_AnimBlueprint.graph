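# Anim Blueprint for the Shiitake Shaolin backpack that adds secondary physics motion to its mushrooms and bag flaps.
# The input pose runs through serial, spring-driven Anim Dynamics nodes, simulating mushrooms bone-relative and flaps root-relative.
# Per-group SelectFloat nodes lower each simulation's alpha while EG_IsSkyDiving is true, damping the motion during skydives.
# The mushroom 03 chain simulates stem_01 twice and has no stem_02 node.
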AnimBlueprint /BRCosmetics/Accessories/FORT_Backpacks/M_MED_Shiitake_Shaolin_Backpack/Meshes/M_MED_Shiitake_Shaolin_Pack_Skeleton_AnimBlueprint
target_skeleton=&/M_MED_Shiitake_Shaolin_Pack_Skeleton
graphs [
 AnimGraph {nodes [
   n1 = SelectFloat(A=0.4, B=0.8, bPickA=Get EG_IsSkyDiving() -> (EG_IsSkyDiving:bool))
   n2 = Anim Dynamics - Bone: dyn_mushroom_01_stem_01(ComponentPose=Local To Component(LocalPose=Input Pose()), Alpha=@n1, ExternalForce="0,0,0") {
    angularDampingOverride=0.9999
    angularSpringConstant=10
    bAngularSpring=true
    bGravityOverrideInSimSpace=true
    bOverrideAngularDamping=true
    simulationSpace=BoneRelative
    alphaBoolBlend {blendOption=Linear}
    boundBone {boneName=dyn_mushroom_01_stem_01}
    componentAppliedLinearAccClamp=<7500 7500 7500>
    componentLinearAccScale=<0.6 0.6 0.6>
    componentLinearVelScale=<0.6 0.6 0.6>
    gravityOverride {z=-980}
    physicsBodyDefinitions [{
      sphereCollisionRadius=0
      boundBone {boneName=dyn_mushroom_01_stem_01}
      boxExtents=<9 18 18>
      constraintSetup {angularLimitsMax=<1 60 60>, angularLimitsMin=<-1 -60 -60>, angularTarget=<0 0 0>}
      localJointOffset=<-1 0 0>
     }]
   }
   n3 = Anim Dynamics - Bone: dyn_mushroom_01_stem_02(ComponentPose=@n2, Alpha=@n1, ExternalForce="0,0,0") {
    angularDampingOverride=0.9999
    angularSpringConstant=25
    bAngularSpring=true
    bGravityOverrideInSimSpace=true
    bOverrideAngularDamping=true
    simulationSpace=BoneRelative
    alphaBoolBlend {blendOption=Linear}
    boundBone {boneName=dyn_mushroom_01_stem_02}
    componentAppliedLinearAccClamp=<7500 7500 7500>
    componentLinearAccScale=<0.6 0.6 0.6>
    componentLinearVelScale=<0.6 0.6 0.6>
    gravityOverride {z=-980}
    physicsBodyDefinitions [{
      sphereCollisionRadius=0
      boundBone {boneName=dyn_mushroom_01_stem_02}
      boxExtents=<9 18 18>
      constraintSetup {angularLimitsMax=<1 60 60>, angularLimitsMin=<-1 -60 -60>, angularTarget=<0 0 0>}
      localJointOffset=<-1 0 0>
     }]
   }
   n4 = Anim Dynamics - Bone: dyn_mushroom_01_stem_03(ComponentPose=@n3, Alpha=@n1, ExternalForce="0,0,0") {
    angularDampingOverride=0.9999
    angularSpringConstant=25
    bAngularSpring=true
    bGravityOverrideInSimSpace=true
    bOverrideAngularDamping=true
    simulationSpace=BoneRelative
    alphaBoolBlend {blendOption=Linear}
    boundBone {boneName=dyn_mushroom_01_stem_03}
    componentAppliedLinearAccClamp=<7500 7500 7500>
    componentLinearAccScale=<0.6 0.6 0.6>
    componentLinearVelScale=<0.6 0.6 0.6>
    gravityOverride {z=-980}
    physicsBodyDefinitions [{
      sphereCollisionRadius=0
      boundBone {boneName=dyn_mushroom_01_stem_03}
      boxExtents=<9 18 18>
      constraintSetup {angularLimitsMax=<1 60 60>, angularLimitsMin=<-1 -60 -60>, angularTarget=<0 0 0>}
      localJointOffset=<-1 0 0>
     }]
   }
   n5 = Anim Dynamics - Bone: dyn_mushroom_01_stem_04(ComponentPose=@n4, Alpha=@n1, ExternalForce="0,0,0") {
    angularDampingOverride=0.9999
    angularSpringConstant=25
    bAngularSpring=true
    bGravityOverrideInSimSpace=true
    bOverrideAngularDamping=true
    simulationSpace=BoneRelative
    alphaBoolBlend {blendOption=Linear}
    boundBone {boneName=dyn_mushroom_01_stem_04}
    componentAppliedLinearAccClamp=<7500 7500 7500>
    componentLinearAccScale=<0.6 0.6 0.6>
    componentLinearVelScale=<0.6 0.6 0.6>
    gravityOverride {z=-980}
    physicsBodyDefinitions [{
      sphereCollisionRadius=0
      boundBone {boneName=dyn_mushroom_01_stem_04}
      boxExtents=<9 18 18>
      constraintSetup {angularLimitsMax=<1 60 60>, angularLimitsMin=<-1 -60 -60>, angularTarget=<0 0 0>}
      localJointOffset=<-1 0 0>
     }]
   }
   n6 = SelectFloat(A=0.23, B=1, bPickA=Get EG_IsSkyDiving())
   n7 = Anim Dynamics - Bone: dyn_mushroom_01_stem_03_fr_01(ComponentPose=@n5, Alpha=@n6, ExternalForce="0,0,0") {
    angularDampingOverride=1
    angularSpringConstant=25
    bAngularSpring=true
    bGravityOverrideInSimSpace=true
    bOverrideAngularDamping=true
    simulationSpace=BoneRelative
    alphaBoolBlend {blendOption=Linear}
    boundBone {boneName=dyn_mushroom_01_stem_03_fr_01}
    componentAppliedLinearAccClamp=<10000 10000 10000>
    componentLinearAccScale=<1.15 1.15 1.15>
    componentLinearVelScale=<1.15 1.15 1.15>
    physicsBodyDefinitions [{
      sphereCollisionRadius=0
      boundBone {boneName=dyn_mushroom_01_stem_03_fr_01}
      boxExtents=<10 20 20>
      constraintSetup {angularLimitsMax=<1 1 15>, angularLimitsMin=<-1 -1 -30>, angularTarget=<0 0 0>}
      localJointOffset=<1 0 0>
     }]
   }
   n8 = Anim Dynamics - Bone: dyn_mushroom_01_stem_03_l_01(ComponentPose=@n7, Alpha=@n6, ExternalForce="0,0,0") {
    angularDampingOverride=1
    angularSpringConstant=25
    bAngularSpring=true
    bGravityOverrideInSimSpace=true
    bOverrideAngularDamping=true
    simulationSpace=BoneRelative
    alphaBoolBlend {blendOption=Linear}
    boundBone {boneName=dyn_mushroom_01_stem_03_l_01}
    componentAppliedLinearAccClamp=<10000 10000 10000>
    componentLinearAccScale=<1.25 1.25 1.25>
    componentLinearVelScale=<1.25 1.25 1.25>
    physicsBodyDefinitions [{
      sphereCollisionRadius=0
      boundBone {boneName=dyn_mushroom_01_stem_03_l_01}
      boxExtents=<10 20 20>
      constraintSetup {angularLimitsMax=<1 1 15>, angularLimitsMin=<-1 -1 -30>, angularTarget=<0 0 0>}
      localJointOffset=<1 0 0>
     }]
   }
   n9 = Anim Dynamics - Bone: dyn_mushroom_01_stem_03_bk_01(ComponentPose=@n8, Alpha=@n6, ExternalForce="0,0,0") {
    angularDampingOverride=1
    angularSpringConstant=25
    bAngularSpring=true
    bGravityOverrideInSimSpace=true
    bOverrideAngularDamping=true
    simulationSpace=BoneRelative
    alphaBoolBlend {blendOption=Linear}
    boundBone {boneName=dyn_mushroom_01_stem_03_bk_01}
    componentAppliedLinearAccClamp=<10000 10000 10000>
    componentLinearAccScale=<1.25 1.25 1.25>
    componentLinearVelScale=<1.25 1.25 1.25>
    physicsBodyDefinitions [{
      sphereCollisionRadius=0
      boundBone {boneName=dyn_mushroom_01_stem_03_bk_01}
      boxExtents=<10 20 20>
      constraintSetup {angularLimitsMax=<1 1 15>, angularLimitsMin=<-1 -1 -30>, angularTarget=<0 0 0>}
      localJointOffset=<1 0 0>
     }]
   }
   n10 = Anim Dynamics - Bone: dyn_mushroom_01_stem_03_r_01(ComponentPose=@n9, Alpha=@n6, ExternalForce="0,0,0") {
    angularDampingOverride=1
    angularSpringConstant=25
    bAngularSpring=true
    bGravityOverrideInSimSpace=true
    bOverrideAngularDamping=true
    simulationSpace=BoneRelative
    alphaBoolBlend {blendOption=Linear}
    boundBone {boneName=dyn_mushroom_01_stem_03_r_01}
    componentAppliedLinearAccClamp=<10000 10000 10000>
    componentLinearAccScale=<1.25 1.25 1.25>
    componentLinearVelScale=<1.25 1.25 1.25>
    physicsBodyDefinitions [{
      sphereCollisionRadius=0
      boundBone {boneName=dyn_mushroom_01_stem_03_r_01}
      boxExtents=<10 20 20>
      constraintSetup {angularLimitsMax=<1 1 15>, angularLimitsMin=<-1 -1 -30>, angularTarget=<0 0 0>}
      localJointOffset=<1 0 0>
     }]
   }
   n11 = SelectFloat(A=0.33, B=0.9, bPickA=Get EG_IsSkyDiving())
   n12 = Anim Dynamics - Bone: dyn_mushroom_02_stem_01(ComponentPose=@n10, Alpha=@n11, ExternalForce="0,0,0") {
    angularDampingOverride=0.9999
    angularSpringConstant=25
    bAngularSpring=true
    bGravityOverrideInSimSpace=true
    bOverrideAngularDamping=true
    simulationSpace=BoneRelative
    alphaBoolBlend {blendOption=Linear}
    boundBone {boneName=dyn_mushroom_02_stem_01}
    componentAppliedLinearAccClamp=<5000 5000 5000>
    componentLinearAccScale=<0.75 0.75 0.75>
    componentLinearVelScale=<0.75 0.75 0.75>
    physicsBodyDefinitions [{
      sphereCollisionRadius=0
      boundBone {boneName=dyn_mushroom_02_stem_01}
      boxExtents=<15 25 25>
      constraintSetup {angularLimitsMax=<1 25 25>, angularLimitsMin=<-1 -25 -25>, angularTarget=<0 0 0>}
      localJointOffset=<-1 -0.5 0>
     }]
   }
   n13 = Anim Dynamics - Bone: dyn_mushroom_02_stem_02(ComponentPose=@n12, Alpha=@n11, ExternalForce="0,0,0") {
    angularDampingOverride=0.9999
    angularSpringConstant=25
    bAngularSpring=true
    bGravityOverrideInSimSpace=true
    bOverrideAngularDamping=true
    simulationSpace=BoneRelative
    alphaBoolBlend {blendOption=Linear}
    boundBone {boneName=dyn_mushroom_02_stem_02}
    componentAppliedLinearAccClamp=<5000 5000 5000>
    componentLinearAccScale=<0.75 0.75 0.75>
    componentLinearVelScale=<0.75 0.75 0.75>
    physicsBodyDefinitions [{
      sphereCollisionRadius=0
      boundBone {boneName=dyn_mushroom_02_stem_02}
      boxExtents=<15 25 25>
      constraintSetup {angularLimitsMax=<1 25 25>, angularLimitsMin=<-1 -25 -25>, angularTarget=<0 0 0>}
      localJointOffset=<-1 -0.5 0>
     }]
   }
   n14 = Anim Dynamics - Bone: dyn_mushroom_02_stem_03(ComponentPose=@n13, Alpha=@n11, ExternalForce="0,0,0") {
    angularDampingOverride=0.9999
    angularSpringConstant=25
    bAngularSpring=true
    bGravityOverrideInSimSpace=true
    bOverrideAngularDamping=true
    simulationSpace=BoneRelative
    alphaBoolBlend {blendOption=Linear}
    boundBone {boneName=dyn_mushroom_02_stem_03}
    componentAppliedLinearAccClamp=<5000 5000 5000>
    componentLinearAccScale=<0.75 0.75 0.75>
    componentLinearVelScale=<0.75 0.75 0.75>
    physicsBodyDefinitions [{
      sphereCollisionRadius=0
      boundBone {boneName=dyn_mushroom_02_stem_03}
      boxExtents=<15 25 25>
      constraintSetup {angularLimitsMax=<1 25 25>, angularLimitsMin=<-1 -25 -25>, angularTarget=<0 0 0>}
      localJointOffset=<-1 -0.5 0>
     }]
   }
   n15 = SelectFloat(A=0.315, B=0.85, bPickA=Get EG_IsSkyDiving())
   n16 = Anim Dynamics - Bone: dyn_mushroom_03_stem_01(ComponentPose=@n14, Alpha=@n15, ExternalForce="0,0,0") {
    angularDampingOverride=0.9999
    angularSpringConstant=30
    bAngularSpring=true
    bGravityOverrideInSimSpace=true
    bOverrideAngularDamping=true
    simulationSpace=BoneRelative
    alphaBoolBlend {blendOption=Linear}
    boundBone {boneName=dyn_mushroom_03_stem_01}
    componentAppliedLinearAccClamp=<5000 5000 5000>
    componentLinearAccScale=<0.75 0.75 0.75>
    componentLinearVelScale=<0.75 0.75 0.75>
    physicsBodyDefinitions [{
      sphereCollisionRadius=0
      boundBone {boneName=dyn_mushroom_03_stem_01}
      boxExtents=<6 12 12>
      constraintSetup {angularLimitsMax=<1 15 15>, angularLimitsMin=<-1 -15 -15>, angularTarget=<0 0 0>}
      localJointOffset=<-1 0.5 0>
     }]
   }
   n17 = Anim Dynamics - Bone: dyn_mushroom_03_stem_01(ComponentPose=@n16, Alpha=@n15, ExternalForce="0,0,0") {
    angularDampingOverride=0.9999
    angularSpringConstant=30
    bAngularSpring=true
    bGravityOverrideInSimSpace=true
    bOverrideAngularDamping=true
    simulationSpace=BoneRelative
    alphaBoolBlend {blendOption=Linear}
    boundBone {boneName=dyn_mushroom_03_stem_01}
    componentAppliedLinearAccClamp=<5000 5000 5000>
    componentLinearAccScale=<0.75 0.75 0.75>
    componentLinearVelScale=<0.75 0.75 0.75>
    physicsBodyDefinitions [{
      sphereCollisionRadius=0
      boundBone {boneName=dyn_mushroom_03_stem_01}
      boxExtents=<6 12 12>
      constraintSetup {angularLimitsMax=<1 15 15>, angularLimitsMin=<-1 -15 -15>, angularTarget=<0 0 0>}
      localJointOffset=<-1 0.5 0>
     }]
   }
   n18 = Anim Dynamics - Bone: dyn_mushroom_03_stem_03(ComponentPose=@n17, Alpha=@n15, ExternalForce="0,0,0") {
    angularDampingOverride=0.9999
    angularSpringConstant=30
    bAngularSpring=true
    bGravityOverrideInSimSpace=true
    bOverrideAngularDamping=true
    simulationSpace=BoneRelative
    alphaBoolBlend {blendOption=Linear}
    boundBone {boneName=dyn_mushroom_03_stem_03}
    componentAppliedLinearAccClamp=<5000 5000 5000>
    componentLinearAccScale=<0.75 0.75 0.75>
    componentLinearVelScale=<0.75 0.75 0.75>
    physicsBodyDefinitions [{
      sphereCollisionRadius=0
      boundBone {boneName=dyn_mushroom_03_stem_03}
      boxExtents=<6 12 12>
      constraintSetup {angularLimitsMax=<1 15 15>, angularLimitsMin=<-1 -15 -15>, angularTarget=<0 0 0>}
      localJointOffset=<-1 0.5 0>
     }]
   }
   n19 = SelectFloat(A=0.3, B=0.8, bPickA=Get EG_IsSkyDiving())
   n20 = Anim Dynamics - Bone: dyn_mushroom_04_stem_01(ComponentPose=@n18, Alpha=@n19, ExternalForce="0,0,0") {
    angularDampingOverride=0.9999
    angularSpringConstant=35
    bAngularSpring=true
    bGravityOverrideInSimSpace=true
    bOverrideAngularDamping=true
    simulationSpace=BoneRelative
    alphaBoolBlend {blendOption=Linear}
    boundBone {boneName=dyn_mushroom_04_stem_01}
    componentAppliedLinearAccClamp=<5000 5000 5000>
    componentLinearAccScale=<0.75 0.75 0.75>
    componentLinearVelScale=<0.75 0.75 0.75>
    physicsBodyDefinitions [{
      sphereCollisionRadius=0
      boundBone {boneName=dyn_mushroom_04_stem_01}
      boxExtents=<9 18 18>
      constraintSetup {angularLimitsMax=<1 25 25>, angularLimitsMin=<-1 -25 -25>, angularTarget=<0 0 0>}
      localJointOffset=<-1 -0.5 0>
     }]
   }
   n21 = Anim Dynamics - Bone: dyn_mushroom_04_stem_02(ComponentPose=@n20, Alpha=@n19, ExternalForce="0,0,0") {
    angularDampingOverride=0.9999
    angularSpringConstant=35
    bAngularSpring=true
    bGravityOverrideInSimSpace=true
    bOverrideAngularDamping=true
    simulationSpace=BoneRelative
    alphaBoolBlend {blendOption=Linear}
    boundBone {boneName=dyn_mushroom_04_stem_02}
    componentAppliedLinearAccClamp=<5000 5000 5000>
    componentLinearAccScale=<0.75 0.75 0.75>
    componentLinearVelScale=<0.75 0.75 0.75>
    physicsBodyDefinitions [{
      sphereCollisionRadius=0
      boundBone {boneName=dyn_mushroom_04_stem_02}
      boxExtents=<9 18 18>
      constraintSetup {angularLimitsMax=<1 25 25>, angularLimitsMin=<-1 -25 -25>, angularTarget=<0 0 0>}
      localJointOffset=<-1 -0.5 0>
     }]
   }
   n22 = Anim Dynamics - Bone: dyn_mushroom_04_stem_03(ComponentPose=@n21, Alpha=@n19, ExternalForce="0,0,0") {
    angularDampingOverride=0.9999
    angularSpringConstant=35
    bAngularSpring=true
    bGravityOverrideInSimSpace=true
    bOverrideAngularDamping=true
    simulationSpace=BoneRelative
    alphaBoolBlend {blendOption=Linear}
    boundBone {boneName=dyn_mushroom_04_stem_03}
    componentAppliedLinearAccClamp=<5000 5000 5000>
    componentLinearAccScale=<0.75 0.75 0.75>
    componentLinearVelScale=<0.75 0.75 0.75>
    physicsBodyDefinitions [{
      sphereCollisionRadius=0
      boundBone {boneName=dyn_mushroom_04_stem_03}
      boxExtents=<9 18 18>
      constraintSetup {angularLimitsMax=<1 25 25>, angularLimitsMin=<-1 -25 -25>, angularTarget=<0 0 0>}
      localJointOffset=<-1 -0.5 0>
     }]
   }
   n23 = SelectFloat(A=0.21, B=0.5, bPickA=Get EG_IsSkyDiving())
   n24 = Anim Dynamics - Bone: dyn_mushroom_05_stem_01(ComponentPose=@n22, Alpha=@n23, ExternalForce="0,0,0") {
    angularDampingOverride=1
    angularSpringConstant=25
    bAngularSpring=true
    bGravityOverrideInSimSpace=true
    bOverrideAngularDamping=true
    simulationSpace=BoneRelative
    alphaBoolBlend {blendOption=Linear}
    boundBone {boneName=dyn_mushroom_05_stem_01}
    componentAppliedLinearAccClamp=<5000 5000 5000>
    componentLinearAccScale=<0.75 1 0.75>
    componentLinearVelScale=<0.75 0.75 0.75>
    physicsBodyDefinitions [{
      sphereCollisionRadius=0
      boundBone {boneName=dyn_mushroom_05_stem_01}
      boxExtents=<4 8 8>
      constraintSetup {angularLimitsMax=<1 15 15>, angularLimitsMin=<-1 -15 -15>, angularTarget=<0 0 0>}
      localJointOffset=<-1 0 -0.5>
     }]
   }
   n25 = Anim Dynamics - Bone: dyn_mushroom_05_stem_02(ComponentPose=@n24, Alpha=@n23, ExternalForce="0,0,0") {
    angularDampingOverride=1
    angularSpringConstant=25
    bAngularSpring=true
    bGravityOverrideInSimSpace=true
    bOverrideAngularDamping=true
    simulationSpace=BoneRelative
    alphaBoolBlend {blendOption=Linear}
    boundBone {boneName=dyn_mushroom_05_stem_02}
    componentAppliedLinearAccClamp=<5000 5000 5000>
    componentLinearAccScale=<0.75 1 0.75>
    componentLinearVelScale=<0.75 0.75 0.75>
    physicsBodyDefinitions [{
      sphereCollisionRadius=0
      boundBone {boneName=dyn_mushroom_05_stem_02}
      boxExtents=<4 8 8>
      constraintSetup {angularLimitsMax=<1 15 15>, angularLimitsMin=<-1 -15 -15>, angularTarget=<0 0 0>}
      localJointOffset=<-1 0 -0.5>
     }]
   }
   n26 = Anim Dynamics - Bone: dyn_mushroom_05_stem_03(ComponentPose=@n25, Alpha=@n23, ExternalForce="0,0,0") {
    angularDampingOverride=1
    angularSpringConstant=25
    bAngularSpring=true
    bGravityOverrideInSimSpace=true
    bOverrideAngularDamping=true
    simulationSpace=BoneRelative
    alphaBoolBlend {blendOption=Linear}
    boundBone {boneName=dyn_mushroom_05_stem_03}
    componentAppliedLinearAccClamp=<5000 5000 5000>
    componentLinearAccScale=<0.75 1 0.75>
    componentLinearVelScale=<0.75 0.75 0.75>
    physicsBodyDefinitions [{
      sphereCollisionRadius=0
      boundBone {boneName=dyn_mushroom_05_stem_03}
      boxExtents=<4 8 8>
      constraintSetup {angularLimitsMax=<1 15 15>, angularLimitsMin=<-1 -15 -15>, angularTarget=<0 0 0>}
      localJointOffset=<-1 0 -0.5>
     }]
   }
   n27 = SelectFloat(A=0.225, B=0.55, bPickA=Get EG_IsSkyDiving())
   n28 = Anim Dynamics - Bone: dyn_mushroom_06_stem_01(ComponentPose=@n26, Alpha=@n27, ExternalForce="0,0,0") {
    angularDampingOverride=1
    angularSpringConstant=30
    bAngularSpring=true
    bGravityOverrideInSimSpace=true
    bOverrideAngularDamping=true
    simulationSpace=BoneRelative
    alphaBoolBlend {blendOption=Linear}
    boundBone {boneName=dyn_mushroom_06_stem_01}
    componentAppliedLinearAccClamp=<5000 5000 5000>
    componentLinearAccScale=<0.75 0.75 0.75>
    componentLinearVelScale=<0.75 0.75 0.75>
    physicsBodyDefinitions [{
      sphereCollisionRadius=0
      boundBone {boneName=dyn_mushroom_06_stem_01}
      boxExtents=<6 11 11>
      constraintSetup {angularLimitsMax=<1 15 15>, angularLimitsMin=<-1 -15 -15>, angularTarget=<0 0 0>}
      localJointOffset=<-1 0.5 0>
     }]
   }
   n29 = Anim Dynamics - Bone: dyn_mushroom_06_stem_02(ComponentPose=@n28, Alpha=@n27, ExternalForce="0,0,0") {
    angularDampingOverride=1
    angularSpringConstant=30
    bAngularSpring=true
    bGravityOverrideInSimSpace=true
    bOverrideAngularDamping=true
    simulationSpace=BoneRelative
    alphaBoolBlend {blendOption=Linear}
    boundBone {boneName=dyn_mushroom_06_stem_02}
    componentAppliedLinearAccClamp=<5000 5000 5000>
    componentLinearAccScale=<0.75 0.75 0.75>
    componentLinearVelScale=<0.75 0.75 0.75>
    physicsBodyDefinitions [{
      sphereCollisionRadius=0
      boundBone {boneName=dyn_mushroom_06_stem_02}
      boxExtents=<6 11 11>
      constraintSetup {angularLimitsMax=<1 15 15>, angularLimitsMin=<-1 -15 -15>, angularTarget=<0 0 0>}
      localJointOffset=<-1 0.5 0>
     }]
   }
   n30 = Anim Dynamics - Bone: dyn_mushroom_06_stem_03(ComponentPose=@n29, Alpha=@n27, ExternalForce="0,0,0") {
    angularDampingOverride=1
    angularSpringConstant=30
    bAngularSpring=true
    bGravityOverrideInSimSpace=true
    bOverrideAngularDamping=true
    simulationSpace=BoneRelative
    alphaBoolBlend {blendOption=Linear}
    boundBone {boneName=dyn_mushroom_06_stem_03}
    componentAppliedLinearAccClamp=<5000 5000 5000>
    componentLinearAccScale=<0.75 0.75 0.75>
    componentLinearVelScale=<0.75 0.75 0.75>
    physicsBodyDefinitions [{
      sphereCollisionRadius=0
      boundBone {boneName=dyn_mushroom_06_stem_03}
      boxExtents=<6 11 11>
      constraintSetup {angularLimitsMax=<1 15 15>, angularLimitsMin=<-1 -15 -15>, angularTarget=<0 0 0>}
      localJointOffset=<-1 0.5 0>
     }]
   }
   n31 = SelectFloat(A=0.24, B=0.6, bPickA=Get EG_IsSkyDiving())
   n32 = Anim Dynamics - Bone: dyn_mushroom_07_stem_01(ComponentPose=@n30, Alpha=@n31, ExternalForce="0,0,0") {
    angularDampingOverride=1
    angularSpringConstant=35
    bAngularSpring=true
    bGravityOverrideInSimSpace=true
    bOverrideAngularDamping=true
    simulationSpace=BoneRelative
    alphaBoolBlend {blendOption=Linear}
    boundBone {boneName=dyn_mushroom_07_stem_01}
    componentAppliedLinearAccClamp=<5000 5000 5000>
    componentLinearAccScale=<0.75 0.75 0.75>
    componentLinearVelScale=<0.75 0.75 0.75>
    physicsBodyDefinitions [{
      sphereCollisionRadius=0
      boundBone {boneName=dyn_mushroom_07_stem_01}
      boxExtents=<5 9 9>
      constraintSetup {angularLimitsMax=<1 15 15>, angularLimitsMin=<-1 -15 -15>, angularTarget=<0 0 0>}
      localJointOffset=<-1 -0.5 0>
     }]
   }
   n33 = Anim Dynamics - Bone: dyn_mushroom_07_stem_02(ComponentPose=@n32, Alpha=@n31, ExternalForce="0,0,0") {
    angularDampingOverride=1
    angularSpringConstant=35
    bAngularSpring=true
    bGravityOverrideInSimSpace=true
    bOverrideAngularDamping=true
    simulationSpace=BoneRelative
    alphaBoolBlend {blendOption=Linear}
    boundBone {boneName=dyn_mushroom_07_stem_02}
    componentAppliedLinearAccClamp=<5000 5000 5000>
    componentLinearAccScale=<0.75 0.75 0.75>
    componentLinearVelScale=<0.75 0.75 0.75>
    physicsBodyDefinitions [{
      sphereCollisionRadius=0
      boundBone {boneName=dyn_mushroom_07_stem_02}
      boxExtents=<5 9 9>
      constraintSetup {angularLimitsMax=<1 15 15>, angularLimitsMin=<-1 -15 -15>, angularTarget=<0 0 0>}
      localJointOffset=<-1 -0.5 0>
     }]
   }
   n34 = Anim Dynamics - Bone: dyn_mushroom_07_stem_03(ComponentPose=@n33, Alpha=@n31, ExternalForce="0,0,0") {
    angularDampingOverride=1
    angularSpringConstant=35
    bAngularSpring=true
    bGravityOverrideInSimSpace=true
    bOverrideAngularDamping=true
    simulationSpace=BoneRelative
    alphaBoolBlend {blendOption=Linear}
    boundBone {boneName=dyn_mushroom_07_stem_03}
    componentAppliedLinearAccClamp=<5000 5000 5000>
    componentLinearAccScale=<0.75 0.75 0.75>
    componentLinearVelScale=<0.75 0.75 0.75>
    physicsBodyDefinitions [{
      sphereCollisionRadius=0
      boundBone {boneName=dyn_mushroom_07_stem_03}
      boxExtents=<5 9 9>
      constraintSetup {angularLimitsMax=<1 15 15>, angularLimitsMin=<-1 -15 -15>, angularTarget=<0 0 0>}
      localJointOffset=<-1 -0.5 0>
     }]
   }
   n35 = SelectFloat(A=0.285, B=0.75, bPickA=Get EG_IsSkyDiving())
   n36 = Anim Dynamics - Bone: dyn_mushroom_08_stem_01(ComponentPose=@n34, Alpha=@n35, ExternalForce="0,0,0") {
    angularDampingOverride=0.9999
    angularSpringConstant=40
    bAngularSpring=true
    bGravityOverrideInSimSpace=true
    bOverrideAngularDamping=true
    simulationSpace=BoneRelative
    alphaBoolBlend {blendOption=Linear}
    boundBone {boneName=dyn_mushroom_08_stem_01}
    componentAppliedLinearAccClamp=<5000 5000 5000>
    componentLinearAccScale=<0.75 0.75 0.75>
    componentLinearVelScale=<0.75 0.75 0.75>
    physicsBodyDefinitions [{
      sphereCollisionRadius=0
      boundBone {boneName=dyn_mushroom_08_stem_01}
      boxExtents=<9 18 18>
      constraintSetup {angularLimitsMax=<1 15 15>, angularLimitsMin=<-1 -15 -15>, angularTarget=<0 0 0>}
      localJointOffset=<-1 0 -0.5>
     }]
   }
   n37 = Anim Dynamics - Bone: dyn_mushroom_08_stem_02(ComponentPose=@n36, Alpha=@n35, ExternalForce="0,0,0") {
    angularDampingOverride=0.9999
    angularSpringConstant=40
    bAngularSpring=true
    bGravityOverrideInSimSpace=true
    bOverrideAngularDamping=true
    simulationSpace=BoneRelative
    alphaBoolBlend {blendOption=Linear}
    boundBone {boneName=dyn_mushroom_08_stem_02}
    componentAppliedLinearAccClamp=<5000 5000 5000>
    componentLinearAccScale=<0.75 0.75 0.75>
    componentLinearVelScale=<0.75 0.75 0.75>
    physicsBodyDefinitions [{
      sphereCollisionRadius=0
      boundBone {boneName=dyn_mushroom_08_stem_02}
      boxExtents=<9 18 18>
      constraintSetup {angularLimitsMax=<1 15 15>, angularLimitsMin=<-1 -15 -15>, angularTarget=<0 0 0>}
      localJointOffset=<-1 0 -0.5>
     }]
   }
   n38 = Anim Dynamics - Bone: dyn_mushroom_08_stem_03(ComponentPose=@n37, Alpha=@n35, ExternalForce="0,0,0") {
    angularDampingOverride=0.9999
    angularSpringConstant=40
    bAngularSpring=true
    bGravityOverrideInSimSpace=true
    bOverrideAngularDamping=true
    simulationSpace=BoneRelative
    alphaBoolBlend {blendOption=Linear}
    boundBone {boneName=dyn_mushroom_08_stem_03}
    componentAppliedLinearAccClamp=<5000 5000 5000>
    componentLinearAccScale=<0.75 0.75 0.75>
    componentLinearVelScale=<0.75 0.75 0.75>
    physicsBodyDefinitions [{
      sphereCollisionRadius=0
      boundBone {boneName=dyn_mushroom_08_stem_03}
      boxExtents=<9 18 18>
      constraintSetup {angularLimitsMax=<1 15 15>, angularLimitsMin=<-1 -15 -15>, angularTarget=<0 0 0>}
      localJointOffset=<-1 0 -0.5>
     }]
   }
   n39 = SelectFloat(A=0.345, B=0.95, bPickA=Get EG_IsSkyDiving())
   n40 = Anim Dynamics - Bone: dyn_mushroom_09_stem_01(ComponentPose=@n38, Alpha=@n39, ExternalForce="0,0,0") {
    angularDampingOverride=0.9999
    angularSpringConstant=20
    bAngularSpring=true
    bGravityOverrideInSimSpace=true
    bOverrideAngularDamping=true
    simulationSpace=BoneRelative
    alphaBoolBlend {blendOption=Linear}
    boundBone {boneName=dyn_mushroom_09_stem_01}
    componentAppliedLinearAccClamp=<5000 5000 5000>
    componentLinearAccScale=<0.75 0.75 0.75>
    componentLinearVelScale=<0.75 0.75 0.75>
    physicsBodyDefinitions [{
      sphereCollisionRadius=0
      boundBone {boneName=dyn_mushroom_09_stem_01}
      boxExtents=<8 14 14>
      constraintSetup {angularLimitsMax=<1 15 15>, angularLimitsMin=<-1 -15 -15>, angularTarget=<0 0 0>}
      localJointOffset=<-1 0.5 0>
     }]
   }
   n41 = Anim Dynamics - Bone: dyn_mushroom_09_stem_02(ComponentPose=@n40, Alpha=@n39, ExternalForce="0,0,0") {
    angularDampingOverride=0.9999
    angularSpringConstant=20
    bAngularSpring=true
    bGravityOverrideInSimSpace=true
    bOverrideAngularDamping=true
    simulationSpace=BoneRelative
    alphaBoolBlend {blendOption=Linear}
    boundBone {boneName=dyn_mushroom_09_stem_02}
    componentAppliedLinearAccClamp=<5000 5000 5000>
    componentLinearAccScale=<0.75 0.75 0.75>
    componentLinearVelScale=<0.75 0.75 0.75>
    physicsBodyDefinitions [{
      sphereCollisionRadius=0
      boundBone {boneName=dyn_mushroom_09_stem_02}
      boxExtents=<8 14 14>
      constraintSetup {angularLimitsMax=<1 15 15>, angularLimitsMin=<-1 -15 -15>, angularTarget=<0 0 0>}
      localJointOffset=<-1 0.5 0>
     }]
   }
   n42 = Anim Dynamics - Bone: dyn_mushroom_09_stem_03(ComponentPose=@n41, Alpha=@n39, ExternalForce="0,0,0") {
    angularDampingOverride=0.9999
    angularSpringConstant=20
    bAngularSpring=true
    bGravityOverrideInSimSpace=true
    bOverrideAngularDamping=true
    simulationSpace=BoneRelative
    alphaBoolBlend {blendOption=Linear}
    boundBone {boneName=dyn_mushroom_09_stem_03}
    componentAppliedLinearAccClamp=<5000 5000 5000>
    componentLinearAccScale=<0.75 0.75 0.75>
    componentLinearVelScale=<0.75 0.75 0.75>
    physicsBodyDefinitions [{
      sphereCollisionRadius=0
      boundBone {boneName=dyn_mushroom_09_stem_03}
      boxExtents=<8 14 14>
      constraintSetup {angularLimitsMax=<1 15 15>, angularLimitsMin=<-1 -15 -15>, angularTarget=<0 0 0>}
      localJointOffset=<-1 0.5 0>
     }]
   }
   n43 = SelectFloat(A=0.285, B=0.9, bPickA=Get EG_IsSkyDiving())
   n44 = Anim Dynamics - Bone: dyn_mushroom_10_stem_01(ComponentPose=@n42, Alpha=@n43, ExternalForce="0,0,0") {
    angularDampingOverride=0.9999
    angularSpringConstant=35
    bAngularSpring=true
    bGravityOverrideInSimSpace=true
    bOverrideAngularDamping=true
    simulationSpace=BoneRelative
    alphaBoolBlend {blendOption=Linear}
    boundBone {boneName=dyn_mushroom_10_stem_01}
    componentAppliedLinearAccClamp=<5000 5000 5000>
    componentLinearAccScale=<0.75 0.75 0.75>
    componentLinearVelScale=<0.75 0.75 0.75>
    physicsBodyDefinitions [{
      sphereCollisionRadius=0
      boundBone {boneName=dyn_mushroom_10_stem_01}
      boxExtents=<8 16 16>
      constraintSetup {angularLimitsMax=<1 20 20>, angularLimitsMin=<-1 -20 -20>, angularTarget=<0 0 0>}
      localJointOffset=<-1 0 0.5>
     }]
   }
   n45 = Anim Dynamics - Bone: dyn_mushroom_10_stem_02(ComponentPose=@n44, Alpha=@n43, ExternalForce="0,0,0") {
    angularDampingOverride=0.9999
    angularSpringConstant=35
    bAngularSpring=true
    bGravityOverrideInSimSpace=true
    bOverrideAngularDamping=true
    simulationSpace=BoneRelative
    alphaBoolBlend {blendOption=Linear}
    boundBone {boneName=dyn_mushroom_10_stem_02}
    componentAppliedLinearAccClamp=<5000 5000 5000>
    componentLinearAccScale=<0.75 0.75 0.75>
    componentLinearVelScale=<0.75 0.75 0.75>
    physicsBodyDefinitions [{
      sphereCollisionRadius=0
      boundBone {boneName=dyn_mushroom_10_stem_02}
      boxExtents=<8 18 18>
      constraintSetup {angularLimitsMax=<1 20 20>, angularLimitsMin=<-1 -20 -20>, angularTarget=<0 0 0>}
      localJointOffset=<-1 0 0.5>
     }]
   }
   n46 = Anim Dynamics - Bone: dyn_mushroom_10_stem_03(ComponentPose=@n45, Alpha=@n43, ExternalForce="0,0,0") {
    angularDampingOverride=0.9999
    angularSpringConstant=35
    bAngularSpring=true
    bGravityOverrideInSimSpace=true
    bOverrideAngularDamping=true
    simulationSpace=BoneRelative
    alphaBoolBlend {blendOption=Linear}
    boundBone {boneName=dyn_mushroom_10_stem_03}
    componentAppliedLinearAccClamp=<5000 5000 5000>
    componentLinearAccScale=<0.75 0.75 0.75>
    componentLinearVelScale=<0.75 0.75 0.75>
    physicsBodyDefinitions [{
      sphereCollisionRadius=0
      boundBone {boneName=dyn_mushroom_10_stem_03}
      boxExtents=<8 18 18>
      constraintSetup {angularLimitsMax=<1 20 20>, angularLimitsMin=<-1 -20 -20>, angularTarget=<0 0 0>}
      localJointOffset=<-1 0 0.5>
     }]
   }
   n47 = SelectFloat(A=0.3, B=1, bPickA=Get EG_IsSkyDiving())
   n48 = Anim Dynamics - Bone: dyn_bag_fr_r_flap_01(ComponentPose=@n46, Alpha=@n47, ExternalForce="0,0,0") {
    angularDampingOverride=0.999
    angularSpringConstant=50
    bAngularSpring=true
    bGravityOverrideInSimSpace=true
    bOverrideAngularDamping=true
    simulationSpace=RootRelative
    alphaBoolBlend {blendOption=Linear}
    boundBone {boneName=dyn_bag_fr_r_flap_01}
    componentAppliedLinearAccClamp=<10000 10000 10000>
    componentLinearAccScale=<1.5 1.5 1.5>
    componentLinearVelScale=<1.5 1.5 1.5>
    physicsBodyDefinitions [{
      sphereCollisionRadius=0
      boundBone {boneName=dyn_bag_fr_r_flap_01}
      boxExtents=<10 20 20>
      constraintSetup {angularLimitsMax=<1 1 45>, angularLimitsMin=<-1 -1 -15>, angularTarget=<0 0 0>}
      localJointOffset=<0.65 0 -2>
     }]
   }
   n49 = Anim Dynamics - Bone: dyn_bag_fr_r_flap_02(ComponentPose=@n48, Alpha=@n47, ExternalForce="0,0,0") {
    angularDampingOverride=0.999
    angularSpringConstant=50
    bAngularSpring=true
    bGravityOverrideInSimSpace=true
    bOverrideAngularDamping=true
    simulationSpace=RootRelative
    alphaBoolBlend {blendOption=Linear}
    boundBone {boneName=dyn_bag_fr_r_flap_02}
    componentAppliedLinearAccClamp=<10000 10000 10000>
    componentLinearAccScale=<1.5 1.5 1.5>
    componentLinearVelScale=<1.5 1.5 1.5>
    physicsBodyDefinitions [{
      sphereCollisionRadius=0
      boundBone {boneName=dyn_bag_fr_r_flap_02}
      boxExtents=<10 20 20>
      constraintSetup {angularLimitsMax=<1 1 30>, angularLimitsMin=<-1 -1 -30>, angularTarget=<0 0 0>}
      localJointOffset=<0.65 0 -2>
     }]
   }
   n50 = SelectFloat(A=0.3, B=1, bPickA=Get EG_IsSkyDiving())
   n51 = Anim Dynamics - Bone: dyn_bag_r_flap_01(ComponentPose=@n49, Alpha=@n50, ExternalForce="0,0,0") {
    angularDampingOverride=0.999
    angularSpringConstant=50
    bAngularSpring=true
    bGravityOverrideInSimSpace=true
    bOverrideAngularDamping=true
    simulationSpace=RootRelative
    alphaBoolBlend {blendOption=Linear}
    boundBone {boneName=dyn_bag_r_flap_01}
    componentAppliedLinearAccClamp=<10000 10000 10000>
    componentLinearAccScale=<1.5 1.5 1.5>
    componentLinearVelScale=<1.5 1.5 1.5>
    physicsBodyDefinitions [{
      sphereCollisionRadius=0
      boundBone {boneName=dyn_bag_r_flap_01}
      boxExtents=<10 20 20>
      constraintSetup {angularLimitsMax=<1 1 45>, angularLimitsMin=<-1 -1 -5>, angularTarget=<0 0 0>}
      localJointOffset=<0.4 0 -2>
     }]
   }
   n52 = Anim Dynamics - Bone: dyn_bag_r_flap_02(ComponentPose=@n51, Alpha=@n50, ExternalForce="0,0,0") {
    angularDampingOverride=0.999
    angularSpringConstant=50
    bAngularSpring=true
    bGravityOverrideInSimSpace=true
    bOverrideAngularDamping=true
    simulationSpace=RootRelative
    alphaBoolBlend {blendOption=Linear}
    boundBone {boneName=dyn_bag_r_flap_02}
    componentAppliedLinearAccClamp=<10000 10000 10000>
    componentLinearAccScale=<1.5 1.5 1.5>
    componentLinearVelScale=<1.5 1.5 1.5>
    physicsBodyDefinitions [{
      sphereCollisionRadius=0
      boundBone {boneName=dyn_bag_r_flap_02}
      boxExtents=<10 20 20>
      constraintSetup {angularLimitsMax=<1 1 30>, angularLimitsMin=<-1 -1 -30>, angularTarget=<0 0 0>}
      localJointOffset=<0.4 0 -2>
     }]
   }
   n53 = SelectFloat(A=0.3, B=1, bPickA=Get EG_IsSkyDiving())
   n54 = Anim Dynamics - Bone: dyn_bag_bk_r_flap_01(ComponentPose=@n52, Alpha=@n53, ExternalForce="0,0,0") {
    angularDampingOverride=0.999
    angularSpringConstant=50
    bAngularSpring=true
    bGravityOverrideInSimSpace=true
    bOverrideAngularDamping=true
    simulationSpace=RootRelative
    alphaBoolBlend {blendOption=Linear}
    boundBone {boneName=dyn_bag_bk_r_flap_01}
    componentAppliedLinearAccClamp=<10000 10000 10000>
    componentLinearAccScale=<1.5 1.5 1.5>
    componentLinearVelScale=<1.5 1.5 1.5>
    physicsBodyDefinitions [{
      sphereCollisionRadius=0
      boundBone {boneName=dyn_bag_bk_r_flap_01}
      boxExtents=<10 20 20>
      constraintSetup {angularLimitsMax=<1 1 45>, angularLimitsMin=<-1 -1 -10>, angularTarget=<0 0 0>}
      localJointOffset=<1 0 -2>
     }]
   }
   n55 = Anim Dynamics - Bone: dyn_bag_bk_r_flap_02(ComponentPose=@n54, Alpha=@n53, ExternalForce="0,0,0") {
    angularDampingOverride=0.999
    angularSpringConstant=50
    bAngularSpring=true
    bGravityOverrideInSimSpace=true
    bOverrideAngularDamping=true
    simulationSpace=RootRelative
    alphaBoolBlend {blendOption=Linear}
    boundBone {boneName=dyn_bag_bk_r_flap_02}
    componentAppliedLinearAccClamp=<10000 10000 10000>
    componentLinearAccScale=<1.5 1.5 1.5>
    componentLinearVelScale=<1.5 1.5 1.5>
    physicsBodyDefinitions [{
      sphereCollisionRadius=0
      boundBone {boneName=dyn_bag_bk_r_flap_02}
      boxExtents=<10 20 20>
      constraintSetup {angularLimitsMax=<1 1 30>, angularLimitsMin=<-1 -1 -30>, angularTarget=<0 0 0>}
      localJointOffset=<1 0 -2>
     }]
   }
   n56 = SelectFloat(A=0.3, B=1, bPickA=Get EG_IsSkyDiving())
   n57 = Anim Dynamics - Bone: dyn_bag_fr_l_flap_01(ComponentPose=@n55, Alpha=@n56, ExternalForce="0,0,0") {
    angularDampingOverride=0.999
    angularSpringConstant=50
    bAngularSpring=true
    bGravityOverrideInSimSpace=true
    bOverrideAngularDamping=true
    simulationSpace=RootRelative
    alphaBoolBlend {blendOption=Linear}
    boundBone {boneName=dyn_bag_fr_l_flap_01}
    componentAppliedLinearAccClamp=<10000 10000 10000>
    componentLinearAccScale=<1.5 1.5 1.5>
    componentLinearVelScale=<1.5 1.5 1.5>
    physicsBodyDefinitions [{
      sphereCollisionRadius=0
      boundBone {boneName=dyn_bag_fr_l_flap_01}
      boxExtents=<10 20 20>
      constraintSetup {angularLimitsMax=<1 1 45>, angularLimitsMin=<-1 -1 -15>, angularTarget=<0 0 0>}
      localJointOffset=<0.65 0 -2>
     }]
   }
   n58 = Anim Dynamics - Bone: dyn_bag_fr_l_flap_02(ComponentPose=@n57, Alpha=@n56, ExternalForce="0,0,0") {
    angularDampingOverride=0.999
    angularSpringConstant=50
    bAngularSpring=true
    bGravityOverrideInSimSpace=true
    bOverrideAngularDamping=true
    simulationSpace=RootRelative
    alphaBoolBlend {blendOption=Linear}
    boundBone {boneName=dyn_bag_fr_l_flap_02}
    componentAppliedLinearAccClamp=<10000 10000 10000>
    componentLinearAccScale=<1.5 1.5 1.5>
    componentLinearVelScale=<1.5 1.5 1.5>
    physicsBodyDefinitions [{
      sphereCollisionRadius=0
      boundBone {boneName=dyn_bag_fr_l_flap_02}
      boxExtents=<10 20 20>
      constraintSetup {angularLimitsMax=<1 1 30>, angularLimitsMin=<-1 -1 -30>, angularTarget=<0 0 0>}
      localJointOffset=<0.65 0 -2>
     }]
   }
   n59 = SelectFloat(A=0.3, B=1, bPickA=Get EG_IsSkyDiving())
   n60 = Anim Dynamics - Bone: dyn_bag_l_flap_01(ComponentPose=@n58, Alpha=@n59, ExternalForce="0,0,0") {
    angularDampingOverride=0.999
    angularSpringConstant=50
    bAngularSpring=true
    bGravityOverrideInSimSpace=true
    bOverrideAngularDamping=true
    simulationSpace=RootRelative
    alphaBoolBlend {blendOption=Linear}
    boundBone {boneName=dyn_bag_l_flap_01}
    componentAppliedLinearAccClamp=<10000 10000 10000>
    componentLinearAccScale=<1.5 1.5 1.5>
    componentLinearVelScale=<1.5 1.5 1.5>
    physicsBodyDefinitions [{
      sphereCollisionRadius=0
      boundBone {boneName=dyn_bag_l_flap_01}
      boxExtents=<10 20 20>
      constraintSetup {angularLimitsMax=<1 1 45>, angularLimitsMin=<-1 -1 -5>, angularTarget=<0 0 0>}
      localJointOffset=<0.4 0 -2>
     }]
   }
   n61 = Anim Dynamics - Bone: dyn_bag_l_flap_02(ComponentPose=@n60, Alpha=@n59, ExternalForce="0,0,0") {
    angularDampingOverride=0.999
    angularSpringConstant=50
    bAngularSpring=true
    bGravityOverrideInSimSpace=true
    bOverrideAngularDamping=true
    simulationSpace=RootRelative
    alphaBoolBlend {blendOption=Linear}
    boundBone {boneName=dyn_bag_l_flap_02}
    componentAppliedLinearAccClamp=<10000 10000 10000>
    componentLinearAccScale=<1.5 1.5 1.5>
    componentLinearVelScale=<1.5 1.5 1.5>
    physicsBodyDefinitions [{
      sphereCollisionRadius=0
      boundBone {boneName=dyn_bag_l_flap_02}
      boxExtents=<10 20 20>
      constraintSetup {angularLimitsMax=<1 1 30>, angularLimitsMin=<-1 -1 -30>, angularTarget=<0 0 0>}
      localJointOffset=<0.4 0 -2>
     }]
   }
   n62 = SelectFloat(A=0.3, B=1, bPickA=Get EG_IsSkyDiving())
   n63 = Anim Dynamics - Bone: dyn_bag_bk_l_flap_01(ComponentPose=@n61, Alpha=@n62, ExternalForce="0,0,0") {
    angularDampingOverride=0.999
    angularSpringConstant=50
    bAngularSpring=true
    bGravityOverrideInSimSpace=true
    bOverrideAngularDamping=true
    simulationSpace=RootRelative
    alphaBoolBlend {blendOption=Linear}
    boundBone {boneName=dyn_bag_bk_l_flap_01}
    componentAppliedLinearAccClamp=<10000 10000 10000>
    componentLinearAccScale=<1.5 1.5 1.5>
    componentLinearVelScale=<1.5 1.5 1.5>
    physicsBodyDefinitions [{
      sphereCollisionRadius=0
      boundBone {boneName=dyn_bag_bk_l_flap_01}
      boxExtents=<10 20 20>
      constraintSetup {angularLimitsMax=<1 1 45>, angularLimitsMin=<-1 -1 -10>, angularTarget=<0 0 0>}
      localJointOffset=<1 0 -2>
     }]
   }
   n64 = Anim Dynamics - Bone: dyn_bag_bk_l_flap_02(ComponentPose=@n63, Alpha=@n62, ExternalForce="0,0,0") {
    angularDampingOverride=0.999
    angularSpringConstant=50
    bAngularSpring=true
    bGravityOverrideInSimSpace=true
    bOverrideAngularDamping=true
    simulationSpace=RootRelative
    alphaBoolBlend {blendOption=Linear}
    boundBone {boneName=dyn_bag_bk_l_flap_02}
    componentAppliedLinearAccClamp=<10000 10000 10000>
    componentLinearAccScale=<1.5 1.5 1.5>
    componentLinearVelScale=<1.5 1.5 1.5>
    physicsBodyDefinitions [{
      sphereCollisionRadius=0
      boundBone {boneName=dyn_bag_bk_l_flap_02}
      boxExtents=<10 20 20>
      constraintSetup {angularLimitsMax=<1 1 30>, angularLimitsMin=<-1 -1 -30>, angularTarget=<0 0 0>}
      localJointOffset=<1 0 -2>
     }]
   }
   n65 = SelectFloat(A=0.3, B=1, bPickA=Get EG_IsSkyDiving())
   n66 = Anim Dynamics - Bone: dyn_bag_fr_flap_01(ComponentPose=@n64, Alpha=@n65, ExternalForce="0,0,0") {
    angularDampingOverride=0.999
    angularSpringConstant=50
    bAngularSpring=true
    bGravityOverrideInSimSpace=true
    bOverrideAngularDamping=true
    simulationSpace=RootRelative
    alphaBoolBlend {blendOption=Linear}
    boundBone {boneName=dyn_bag_fr_flap_01}
    componentAppliedLinearAccClamp=<10000 10000 10000>
    componentLinearAccScale=<1.5 1.5 1.5>
    componentLinearVelScale=<1.5 1.5 1.5>
    physicsBodyDefinitions [{
      sphereCollisionRadius=0
      boundBone {boneName=dyn_bag_fr_flap_01}
      boxExtents=<10 20 20>
      constraintSetup {angularLimitsMax=<1 1 45>, angularLimitsMin=<-1 -1 -15>, angularTarget=<0 0 0>}
      localJointOffset=<0.5 0 -2>
     }]
   }
   n67 = Anim Dynamics - Bone: dyn_bag_fr_flap_02(ComponentPose=@n66, Alpha=@n65, ExternalForce="0,0,0") {
    angularDampingOverride=0.999
    angularSpringConstant=50
    bAngularSpring=true
    bGravityOverrideInSimSpace=true
    bOverrideAngularDamping=true
    simulationSpace=RootRelative
    alphaBoolBlend {blendOption=Linear}
    boundBone {boneName=dyn_bag_fr_flap_02}
    componentAppliedLinearAccClamp=<10000 10000 10000>
    componentLinearAccScale=<1.5 1.5 1.5>
    componentLinearVelScale=<1.5 1.5 1.5>
    physicsBodyDefinitions [{
      sphereCollisionRadius=0
      boundBone {boneName=dyn_bag_fr_flap_02}
      boxExtents=<10 20 20>
      constraintSetup {angularLimitsMax=<1 1 30>, angularLimitsMin=<-1 -1 -30>, angularTarget=<0 0 0>}
      localJointOffset=<0.5 0 -2>
     }]
   }
   n68 = SelectFloat(A=0.3, B=1, bPickA=Get EG_IsSkyDiving())
   n69 = Anim Dynamics - Bone: dyn_bag_bk_flap_01(ComponentPose=@n67, Alpha=@n68, ExternalForce="0,0,0") {
    angularDampingOverride=0.999
    angularSpringConstant=50
    bAngularSpring=true
    bGravityOverrideInSimSpace=true
    bOverrideAngularDamping=true
    simulationSpace=RootRelative
    alphaBoolBlend {blendOption=Linear}
    boundBone {boneName=dyn_bag_bk_flap_01}
    componentAppliedLinearAccClamp=<10000 10000 10000>
    componentLinearAccScale=<1.5 1.5 1.5>
    componentLinearVelScale=<1.5 1.5 1.5>
    physicsBodyDefinitions [{
      sphereCollisionRadius=0
      boundBone {boneName=dyn_bag_bk_flap_01}
      boxExtents=<10 20 20>
      constraintSetup {angularLimitsMax=<1 1 45>, angularLimitsMin=<-1 -1 -15>, angularTarget=<0 0 0>}
      localJointOffset=<1 0 -2>
     }]
   }
   n70 = Anim Dynamics - Bone: dyn_bag_bk_flap_02(ComponentPose=@n69, Alpha=@n68, ExternalForce="0,0,0") {
    angularDampingOverride=0.999
    angularSpringConstant=50
    bAngularSpring=true
    bGravityOverrideInSimSpace=true
    bOverrideAngularDamping=true
    simulationSpace=RootRelative
    alphaBoolBlend {blendOption=Linear}
    boundBone {boneName=dyn_bag_bk_flap_02}
    componentAppliedLinearAccClamp=<10000 10000 10000>
    componentLinearAccScale=<1.5 1.5 1.5>
    componentLinearVelScale=<1.5 1.5 1.5>
    physicsBodyDefinitions [{
      sphereCollisionRadius=0
      boundBone {boneName=dyn_bag_bk_flap_02}
      boxExtents=<10 20 20>
      constraintSetup {angularLimitsMax=<1 1 30>, angularLimitsMin=<-1 -1 -30>, angularTarget=<0 0 0>}
      localJointOffset=<1 0 -2>
     }]
   }
   n71 = SelectFloat(A=0.3, B=1, bPickA=Get EG_IsSkyDiving())
   n72 = Anim Dynamics - Bone: dyn_pouch_l_01(ComponentPose=@n70, Alpha=@n71, ExternalForce="0,0,0") {
    angularDampingOverride=0.9999
    angularSpringConstant=10
    bAngularSpring=true
    bGravityOverrideInSimSpace=true
    bLinearSpring=true
    bOverrideAngularDamping=true
    linearSpringConstant=10
    simulationSpace=RootRelative
    alphaBoolBlend {blendOption=Linear}
    boundBone {boneName=dyn_pouch_l_01}
    componentAppliedLinearAccClamp=<5000 5000 5000>
    componentLinearAccScale=<0.5 0.5 0.5>
    componentLinearVelScale=<0.5 0.5 0.5>
    physicsBodyDefinitions [{
      sphereCollisionRadius=0
      boundBone {boneName=dyn_pouch_l_01}
      boxExtents=<20 10 20>
      constraintSetup {angularLimitsMax=<1 1 0>, angularLimitsMin=<-1 -1 -25>, angularTarget=<0 0 0>, linearAxesMax=<0.5 0 0>, linearAxesMin=<-0.5 0 0>}
      localJointOffset=<-1 0 0>
     }]
   }
   n73 = Anim Dynamics - Bone: dyn_pouch_l_02(ComponentPose=@n72, Alpha=@n71, ExternalForce="0,0,0") {
    angularDampingOverride=0.9999
    angularSpringConstant=10
    bAngularSpring=true
    bGravityOverrideInSimSpace=true
    bLinearSpring=true
    bOverrideAngularDamping=true
    linearSpringConstant=10
    simulationSpace=RootRelative
    alphaBoolBlend {blendOption=Linear}
    boundBone {boneName=dyn_pouch_l_02}
    componentAppliedLinearAccClamp=<10000 10000 10000>
    componentLinearAccScale=<0.5 0.5 0.5>
    componentLinearVelScale=<0.5 0.5 0.5>
    physicsBodyDefinitions [{
      sphereCollisionRadius=0
      boundBone {boneName=dyn_pouch_l_02}
      boxExtents=<20 10 20>
      constraintSetup {angularLimitsMax=<1 1 30>, angularLimitsMin=<-1 -1 -30>, angularTarget=<0 0 0>, linearAxesMax=<0.5 0 0>, linearAxesMin=<-0.5 0 0>}
      localJointOffset=<-1 0 -2>
     }]
   }
   n74 = Anim Dynamics - Bone: dyn_pouch_r_01(ComponentPose=@n73, Alpha=@n71, ExternalForce="0,0,0") {
    angularDampingOverride=0.9999
    angularSpringConstant=10
    bAngularSpring=true
    bGravityOverrideInSimSpace=true
    bLinearSpring=true
    bOverrideAngularDamping=true
    linearSpringConstant=10
    simulationSpace=RootRelative
    alphaBoolBlend {blendOption=Linear}
    boundBone {boneName=dyn_pouch_r_01}
    componentAppliedLinearAccClamp=<5000 5000 5000>
    componentLinearAccScale=<0.5 0.5 0.5>
    componentLinearVelScale=<0.5 0.5 0.5>
    physicsBodyDefinitions [{
      sphereCollisionRadius=0
      boundBone {boneName=dyn_pouch_r_01}
      boxExtents=<20 10 20>
      constraintSetup {angularLimitsMax=<1 1 25>, angularLimitsMin=<-1 -1 0>, angularTarget=<0 0 0>, linearAxesMax=<0.5 0 0>, linearAxesMin=<-0.5 0 0>}
      localJointOffset=<-1 0 0>
     }]
   }
   n75 = Anim Dynamics - Bone: dyn_pouch_r_02(ComponentPose=@n74, Alpha=@n71, ExternalForce="0,0,0") {
    angularDampingOverride=0.9999
    angularSpringConstant=10
    bAngularSpring=true
    bGravityOverrideInSimSpace=true
    bLinearSpring=true
    bOverrideAngularDamping=true
    linearSpringConstant=10
    simulationSpace=RootRelative
    alphaBoolBlend {blendOption=Linear}
    boundBone {boneName=dyn_pouch_r_02}
    componentAppliedLinearAccClamp=<10000 10000 10000>
    componentLinearAccScale=<0.5 0.5 0.5>
    componentLinearVelScale=<0.5 0.5 0.5>
    physicsBodyDefinitions [{
      sphereCollisionRadius=0
      boundBone {boneName=dyn_pouch_r_02}
      boxExtents=<10 20 20>
      constraintSetup {angularLimitsMax=<1 1 30>, angularLimitsMin=<-1 -1 -30>, angularTarget=<0 0 0>, linearAxesMax=<0.5 0 0>, linearAxesMin=<-0.5 0 0>}
      localJointOffset=<-1 0 -2>
     }]
   }
   Output Pose(Result=Component To Local(ComponentPose=@n75)) {layerGroup=None}
  ]}
 EventGraph {nodes [
   Event BlueprintUpdateAnimation()
   > n1 = Branch(Condition=Get bIsSkyDiving())
   > Set EG_IsSkyDiving(EG_IsSkyDiving:bool=true) -> (Output_Get:bool)
   Set EG_IsSkyDiving(execute=@n1.else, EG_IsSkyDiving=false)
   TryGetPawnOwner()
  ]}
]
variables [EG_IsSkyDiving {type=boolean}]
